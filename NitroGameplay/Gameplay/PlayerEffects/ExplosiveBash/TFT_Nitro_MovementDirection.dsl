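; Targeting filter for Nitro's Explosive Bash that keeps only targets lying within a configured angular band relative to the source actor's direction of movement.
; It takes the dot product of the source's normalized horizontal velocity and the horizontal direction to the target's bounds origin.
; Targets with a dot outside the MinDot to MaxDot range, or with an invalid hit actor, are removed.
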(fn ShouldRemoveTarget
  (params (TargetingHandle :type Targeting%20Request%20Handle%20Structure%20%28by%20ref%29) (TargetData :type Targeting%20Default%20Result%20Data%20Structure%20%28by%20ref%29))
  (bind (_bblockinghit _binitialoverlap _time _distance _location _impactpoint _normal _impactnormal _physmat _hitactor) (Collision|BreakHitResult (Utilities|Struct|BreakTargetingDefaultResultData TargetData)))
  (bind _output_get (LocalVariables|SetCurrentTarget _hitactor))
  (Utilities|IsValid _output_get
    (:Is%20Valid
      (bind _output_get_2 (LocalVariables|SetSourceActor (Utilities|Struct|BreakTargetingSourceContext (TargetingSystem|TargetingTypes|GetTargetingSourceContext (Targeting|GetTargetingSubsystem (Variables|GetTargetingHandle)) (Variables|GetTargetingHandle)))))
      (LocalVariables|SetVelocityDirectionXY (CallFunction|GetVelocityDirection _output_get_2))
      (LocalVariables|SetDirectionToTargetXY (CallFunction|GetDirectiontoTarget (LocalVariables|GetSourceActor) (LocalVariables|GetCurrentTarget)))
      (return (not (Math|Float|InRange(Float) (Math|Vector|DotProduct (LocalVariables|GetVelocityDirectionXY) (LocalVariables|GetDirectionToTargetXY)) (Variables|Default|GetMinDot) (Variables|Default|GetMaxDot)))))
    (:Is%20Not%20Valid
      (return true))))

(fn GetDirectionToTarget
  (params (SourceActor :type Actor%20Object%20Reference) (TargetActor :type Actor%20Object%20Reference))
  (return (Math|Vector|Normalize (* (Math|Vector|GetUnitDirection(Vector) (Transformation|GetActorLocation SourceActor) (CallFunction|GetTargetLocation TargetActor)) "1.000000,1.000000,0.000000"))))

(fn GetTargetLocation
  (params (TargetActor :type Actor%20Object%20Reference))
  (return (Collision|GetActorBounds TargetActor false false)))

(fn GetVelocityDirection
  (params (SourceActor :type Actor%20Object%20Reference))
  (return (Math|Vector|Normalize (* (Transformation|GetVelocity SourceActor) "1.000000,1.000000,0.000000"))))
