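; Targeting filter for Nitro Explosive Bash that removes targets whose surface is walkable for the source, so walkable slopes are not bashed.
; Skip-check classes or tag, and an optional only-check list (testing just its first entry), exempt targets.
; Others get a line trace, offset for vehicles and ignoring passengers; a miss or different actor removes them.
; The hit normal's angle is compared to the pawn's buffered walkable floor angle, with debug printing.
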(fn ShouldRemoveTarget
  (params (TargetingHandle :type Targeting%20Request%20Handle%20Structure%20%28by%20ref%29) (TargetData :type Targeting%20Default%20Result%20Data%20Structure%20%28by%20ref%29))
  (bind _hitresult (Utilities|Struct|BreakTargetingDefaultResultData TargetData))
  (bind _returnvalue (Targeting|GetTargetingSubsystem TargetingHandle))
  (bind _returnvalue_2 (TargetingSystem|TargetingTypes|GetTargetingSourceContext _returnvalue TargetingHandle))
  (bind _sourceactor (Utilities|Struct|BreakTargetingSourceContext _returnvalue_2))
  (bind (_bblockinghit _binitialoverlap _time _distance _location _impactpoint _normal _impactnormal _physmat _hitactor) (Collision|BreakHitResult _hitresult))
  (bind (_impactnormal_x _impactnormal_y _impactnormal_z) (Math|Vector|BreakVector (Variables|GetImpactNormal_X)))
  (Utilities|IsValid _hitactor
    (:Is%20Valid
      (bind _output_get (LocalVariables|SetCurrentTarget _hitactor))
      ;[1 1 1] Don't apply this filter if we hit a bash-able object
      (if (not (CallFunction|IsASkipCheckClass (Utilities|GetClass _output_get)))
        (bind _returnvalue_3 (GameplayTags|HasMatchingGameplayTag(Message) (LocalVariables|GetCurrentTarget) "(TagName=\"Nitro.Bash.SkipCheck\")"))
        (if (not _returnvalue_3)
          (if (Utilities|Array|IsEmpty (Variables|Default|GetOnlyCheckClasses))
            _
            (elif (CallFunction|IsAnOnlyCheckClass (Utilities|GetClass _output_get))
              _
              (else
                (return false))))
          (bind _outhit (CallFunction|TracefromSource _sourceactor (LocalVariables|GetCurrentTarget)))
          (LocalVariables|SetImpactNormal (.ImpactNormal (Collision|BreakHitResult _outhit)))
          (if (Collision|BreakHitResult _outhit)
            (if (== (.HitActor (Collision|BreakHitResult _outhit)) (LocalVariables|GetCurrentTarget))
              (bind _returnvalue_4 (CallFunction|IsAngleWalkable _sourceactor _impactnormal_z))
              (return _returnvalue_4)
              (else
                (return true)))
            (else
              (return true)))))
      (return false))
    (:Is%20Not%20Valid
      (return true))))

(fn GetTargetLocation
  (params (SourceActor :type Actor%20Object%20Reference) (TargetActor :type Actor%20Object%20Reference) (SourceLocation :type Vector))
  (if (Variables|Default|GetbTraceToBoundsCenter)
    (bind _asbuilding_actor (Utilities|Casting|CastToBuildingActor TargetActor)
      (:then
        (return (Fortnite|GetActorCollisionComponentBounds _asbuilding_actor)))
      (:CastFailed
        (return (Transformation|GetActorLocation TargetActor))))
    (else
      (return (+ (Variables|GetSourceLocation) (* (Transformation|GetActorForwardVector (Variables|GetSourceActor)) (Variables|Default|GetTraceDistance)))))))

(fn GetTraceActorsToIgnore
  (params (SourceActor :type Actor%20Object%20Reference))
  (LocalVariables|SetActorArray (Utilities|Array|MakeArray SourceActor))
  (bind _asfort_vehicle_interface (Utilities|Casting|CastToFortVehicleInterface (Variables|GetSourceActor))
    (:then
      (Utilities|Array|AppendArray (LocalVariables|GetActorArray) (Vehicle|GetAllPassengers _asfort_vehicle_interface)))
    (:CastFailed
      _))
  (return (LocalVariables|GetActorArray)))

(fn GetTraceSourceLocation
  (params (SourceActor :type Actor%20Object%20Reference))
  (bind _asfort_athena_vehicle (Utilities|Casting|CastToFortAthenaVehicle SourceActor)
    (:then
      ;[1 1 1] Vehicle root location is super low, so we will often hit the ground when performing the trace. This pops it up a bit to be more accurate. Maybe we should use a specific socket location to be more flexible with the different vehicle types
      (return (+ (Transformation|GetActorLocation _asfort_athena_vehicle) (Variables|Default|GetVehicleSourceOffset))))
    (:CastFailed
      (return (Transformation|GetActorLocation SourceActor)))))

(fn GetWalkableFloorAngle
  (params (Object :type Object%20Reference))
  (bind _asfort_pawn (Utilities|Casting|CastToFortPawn Object)
    (:then
      (return (+ (Pawn|Components|CharacterMovement|GetWalkableFloorAngle (Variables|/Script/Engine.Character|GetCharacterMovement _asfort_pawn)) (Variables|Default|GetWalkableAngleBuffer))))
    (:CastFailed
      (return 55.0))))

(fn IsAngleWalkable
  (params (object :type Object%20Reference) (NormalZ :type Float%20%28double-precision%29))
  (bind _returnvalue (Math|Trig|Acos(Degrees) NormalZ))
  (bind _returnvalue_2 (CallFunction|GetWalkableFloorAngle object))
  (Development|PrintString 0 (Utilities|String|ToString(Text) (Utilities|Text|FormatText "NSLOCTEXT(\"[ED745771AFE7B4831379C7E61EF4E93D]\", \"2B8D58D24B82BB8D06E4F3A9D45D80CD\", \"Target walkable: {Walkable} Angle: {Angle}\")" (<= _returnvalue _returnvalue_2) (Utilities|String|ToString(Float) _returnvalue))) true true "(R=0.000000,G=0.660000,B=1.000000,A=1.000000)" 5.0)
  (return (<= _returnvalue _returnvalue_2)))

(fn IsAnOnlyCheckClass
  (params (Class :type Actor%20Class%20Reference%20%28by%20ref%29))
  (for _array_element (Variables|Default|GetOnlyCheckClasses)
    (return (Utilities|ClassIsChildOf Class _array_element)))
  (return false))

(fn IsASkipCheckClass
  (params (Class :type Actor%20Class%20Reference%20%28by%20ref%29))
  (for _array_element (Variables|Default|GetSkipCheckClasses)
    (if (Utilities|ClassIsChildOf Class _array_element)
      (return true)))
  (return false))

(fn TraceFromSource
  (params (SourceActor :type Actor%20Object%20Reference) (TargetActor :type Actor%20Object%20Reference))
  (bind _output_get (LocalVariables|SetSourceLoc (CallFunction|GetTraceSourceLocation SourceActor)))
  (bind _outlocation (CallFunction|GetTargetLocation SourceActor TargetActor _output_get))
  (bind _outhit (Collision|LineTraceForObjects SourceActor (LocalVariables|GetSourceLoc) _outlocation (Variables|Default|GetObject%20Types) false (CallFunction|GetTraceActorstoIgnore (Variables|GetSourceActor)) "ForDuration"))
  (return _outhit))
