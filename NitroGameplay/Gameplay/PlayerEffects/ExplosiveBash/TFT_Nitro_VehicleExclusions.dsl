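; Targeting filter task in the Nitro Explosive Bash player effect that drops building actors from a targeting request's results.
; ShouldRemoveTarget casts each target's hit actor to BuildingActor, returning true to remove it on success and false to keep it otherwise.
; Despite the VehicleExclusions name, the function contains no vehicle-specific check.
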(fn ShouldRemoveTarget
  (params (TargetingHandle :type Targeting%20Request%20Handle%20Structure%20%28by%20ref%29) (TargetData_HitResult_bBlockingHit :type Boolean) (TargetData_HitResult_bInitialOverlap :type Boolean) (TargetData_HitResult_Time :type Float%20%28single-precision%29) (TargetData_HitResult_Distance :type Float%20%28single-precision%29) (TargetData_HitResult_Location :type Vector :default "0, 0, 0") (TargetData_HitResult_ImpactPoint :type Vector :default "0, 0, 0") (TargetData_HitResult_Normal :type Vector :default "0, 0, 0") (TargetData_HitResult_ImpactNormal :type Vector :default "0, 0, 0") (TargetData_HitResult_PhysMat :type Physical%20Material%20Object%20Reference) (TargetData_HitResult_HitActor :type Actor%20Object%20Reference) (TargetData_HitResult_HitComponent :type Primitive%20Component%20Object%20Reference) (TargetData_HitResult_HitBoneName :type Name :default "None") (TargetData_HitResult_BoneName :type Name :default "None") (TargetData_HitResult_HitItem :type Integer) (TargetData_HitResult_ElementIndex :type Integer) (TargetData_HitResult_FaceIndex :type Integer) (TargetData_HitResult_TraceStart :type Vector :default "0, 0, 0") (TargetData_HitResult_TraceEnd :type Vector :default "0, 0, 0") (TargetData_Score :type Float%20%28single-precision%29))
  (Utilities|Casting|CastToBuildingActor TargetData_HitResult_HitActor
    (:then
      (return true))
    (:CastFailed
      (return false))))
